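; Weapon Blueprint for the WiseOtter repulsors, handling overheat feedback, lock-on aiming and alternating-hand muzzle locations.
; Overheat start plays FX and sound and applies an active gameplay effect; cooling applies finished and removal effects, while unequip clears the state.
; An aim-rotation override, bound at BeginPlay, turns camera-sourced aim toward the lock-on target's wildlife capsule center, pawn chest socket or actor location.
; Muzzle location follows the last hand fired, and most failed actions play an error ping.
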(event Overheating|EventOnStartOverheated
  ;[1 1 1] Overheat Started
  (|CallParentFunction)
  (Effects|SpawnFXSystemAttached "/WiseOtter/Assets/FX/NS_WiseOtter_Repulsors_Overheat.NS_WiseOtter_Repulsors_Overheat" (Variables|/Script/Engine.Character|GetMesh (Utilities|Casting|CastToFortPawn (Variables|Actor|GetInstigator))) "root")
  (Audio|Components|Audio|Play (Variables|Default|GetOnOverheatSound) 0.0)
  (Variables|Default|SetTriggerOverheatFinishedGC true)
  (CallFunction|SafeApplyGE (Variables|Default|GetGE_Overheat_Active)))

(event Overheating|EventOnFinishedCooling
  ;[1 1 1] Overheat Finished Cooling
  (|CallParentFunction)
  (if (Variables|Default|GetTriggerOverheatFinishedGC)
    (Variables|Default|SetTriggerOverheatFinishedGC false)
    (CallFunction|SafeApplyGE (Variables|Default|GetGE_Overheat_Finished))
    (CallFunction|SafeApplyGE (Variables|Default|GetGE_Overheat_Remove_Active))))

(event FortWeapon|EventOnLocalUnEquip
  ;[1 1 1] On Local Unequip
  (|CallParentFunction)
  (Audio|Parameter|SetFloatParameter (Variables|Default|GetOverheatLoopSound) "Overheat" 0.0 :!class /Script/AudioExtensions.AudioParameterControllerInterface)
  (Variables|Default|SetTriggerOverheatFinishedGC false)
  (CallFunction|SafeApplyGE (Variables|Default|GetGE_Overheat_Remove_Active)))

(event EventBeginPlay
  (|CallParentFunction)
  (Variables|FortWeapon|SetOnGetAimRotOverride (EventDispatchers|CreateEvent :!event OnGetAimRot)))

(event EventOnUnableToPerformAction
  (params (FailedReason :type Gameplay%20Tag%20Container%20Structure%20%28by%20ref%29) (LocalizedFailureReasonText :type Text%20%28by%20ref%29))
  ; Unable to fire audio
  (if (== (Utilities|Array|FindItem (GameplayTags|BreakGameplayTagContainer FailedReason) "(TagName=\"Activate.Failed.CanActivateAbility\")") -1)
    (Audio|PlaySound2D "/FNGameplayCues/Sounds/Fort_UI_Sounds/Gameplay/UI_Error_Ping_Cant_Do_Default_Cue.UI_Error_Ping_Cant_Do_Default_Cue")))

(fn GetBulletTargetLocation
  (bind _lockontarget (Variables|Default|GetLockOnTarget))
  (Utilities|FlowControl|Sequence
    (:then_0
      (Utilities|IsValid _lockontarget
        (:Is%20Valid
          (Utilities|FlowControl|Sequence
            (:then_0
              (bind _asnpc_pawn_wildlife_parent (Utilities|Casting|CastToNPC_Pawn_Wildlife_Parent _lockontarget)
                (:then
                  (Utilities|IsValid (Variables|/Script/Engine.Character|GetCapsuleComponent _asnpc_pawn_wildlife_parent)
                    (:Is%20Valid
                      (return (Physics|GetCenterOfMass (Variables|/Script/Engine.Character|GetCapsuleComponent _asnpc_pawn_wildlife_parent) "None")))))
                (:CastFailed
                  (bind _asfort_pawn (Utilities|Casting|CastToFortPawn _lockontarget))
                  (Utilities|IsValid (Variables|/Script/Engine.Character|GetMesh _asfort_pawn)
                    (:Is%20Valid
                      (return (Transformation|GetSocketLocation (Variables|/Script/Engine.Character|GetMesh _asfort_pawn) "FX_Chest")))))))
            (:then_1
              ;[1 0.4635 0.4635] We have a valid target, but did not find a good output location. Return the target actor's location.
              (return (Transformation|GetActorLocation _lockontarget)))))))
    (:then_1
      ;[1 0.4635 0.4635] Complete failure, return zero vector.
      (return "0, 0, 0"))))

(fn GetMuzzleLocation
  (params (PatternIndex :type Integer))
  (bind _asfort_pawn (Utilities|Casting|CastToFortPawn (Game|GetInstigator)))
  (return (Transformation|GetSocketLocation (Variables|/Script/Engine.Character|GetMesh _asfort_pawn) (Utilities|Select :Index (Variables|Default|GetLastHandFired) :LEFT "LeftHand" :RIGHT "RightHand"))))

(fn OnGetAimRot
  (params (Weapon :type Fort%20Weapon%20Object%20Reference) (TargetingSource :type EFortAbilityTargetingSource%20Enum :default "Camera"))
  (bind _returnvalue_location (Math|Transform|BreakTransform (Targeting|GetTargetingTransform (Utilities|Casting|CastToFortPawn (Variables|Actor|GetInstigator)) "Camera" 0)))
  (bind _instigator (Variables|Actor|GetInstigator))
  (bind (_asfort_pawn _bsuccess) (Utilities|Casting|CastToFortPawn _instigator))
  (if (and (Utilities|Enum|Equal(Enum) TargetingSource "Camera") (Utilities|IsValid _instigator :!class /Script/Engine.KismetSystemLibrary))
    (if (and (Utilities|IsValid (Variables|Default|GetLockOnTarget) :!class /Script/Engine.KismetSystemLibrary) _bsuccess)
      ;[0 0.5781 0.02235] Rotate aim toward target if locked on.
      (bind _outlocation (Default|GetBulletTargetLocation))
      (return (Math|Rotator|FindLookatRotation _returnvalue_location _outlocation) true)
      (else
        _))
    (else
      _))
  (return "0,0.000000,1.000000" false))

(fn SafeApplyGE
  (params (GE :type Gameplay%20Effect%20Class%20Reference))
  (Utilities|IsValid (Game|GetInstigator)
    (:Is%20Valid
      (Utilities|IsValid (Ability|GetAbilitySystemComponent (Game|GetInstigator))
        (:Is%20Valid
          (Fortnite|ApplyGameplayEffectsToActor (Game|GetInstigator) (Utilities|Array|MakeArray (Variables|GetGE)) (Transformation|GetActorLocation (Game|GetInstigator)) false 0 (Ability|GetAbilitySystemComponent (Game|GetInstigator))))))))

(fn ConstructionScript)
