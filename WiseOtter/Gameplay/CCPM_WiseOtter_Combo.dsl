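; Custom character part blueprint for WiseOtter that adds idle and muzzle repulsor Niagara effects to the player's skeletal mesh sockets.
; On pawn or part change, it attaches them and multicasts the current weapon to bind OnHandFired, with a delay hack for late weapon validity on laggy clients.
; Firing, unless an active-power tag is present, activates that hand's muzzle effect, sets emissive color and shakes the camera.
; An OnRep function also binds Ashton Indigo weapons.
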(event EventOnPawnAndPartChanged
  (bind _returnvalue (CustomCharacter|GetAssociatedPlayerPawn))
  ;[1 1 1] New
  (|CallParentFunction)
  (CallFunction|MuzzleAttachment)
  (CallFunction|IdleAttachment)
  (Utilities|IsValid _returnvalue
    (:Is%20Valid
      (CallFunction|MuzzleFiringSetup (Weapon|GetCurrentWeapon _returnvalue)))))

(event Custom|Idle%20Attachment
  (bind _returnvalue (CustomCharacter|GetCharacterPartType))
  (bind _returnvalue_2 (CustomCharacter|GetAssociatedPlayerPawn))
  (bind _returnvalue_3 (Customization|GetSkeletalMeshforPartType _returnvalue_2 _returnvalue :!class /Script/FortniteGame.FortPlayerPawn))
  (Transformation|AttachComponentToComponent (Variables|Default|GetNS_HIghTower_Tomato_FlareIdle_L) _returnvalue_3 "FX_repulsor_l" "KeepRelative" "KeepRelative" "KeepRelative" true)
  (Transformation|AttachComponentToComponent (Variables|Default|GetNS_HIghTower_Tomato_FlareIdle_R) _returnvalue_3 "FX_repulsor_r" "KeepRelative" "KeepRelative" "KeepRelative" true)
  ;[1 1 1] Cancel Auto Activate of Idle FX, add set active to fix JIRA FORT-308838
  (CallFunction|IsHiddinginProp (GameplayTags|HasMatchingGameplayTag (CustomCharacter|GetAssociatedPlayerPawn) "(TagName=\"Granted.Athena.EnvItem.HidingProp.HidingInProp\")")))

(event Custom|Muzzle_Firing_Multicast
  (params (Weapon :type Fort%20Weapon%20Object%20Reference))
  ;[1 0.6369 0.427] Muzzle Firing Setup - Must be Replicated to Function!
  (bind _asb_high_tower_tomato_repulsors (Utilities|Casting|CastToB_HighTower_Tomato_Repulsors Weapon)
    (:then
      _)
    (:CastFailed
      (Development|PrintString "CASTFAILED" true true "(R=1.000000,G=0.000000,B=0.031557,A=1.000000)")
      ;[1 0 0] Hack loop to fix current weapon not being valid yet on laggy clients.
      (Utilities|FlowControl|DoN 5
        (:Exit
          (Utilities|FlowControl|Delay 0.1)))))
  (RangedWeaponDual|BindEventtoOnHandFired _asb_high_tower_tomato_repulsors (EventDispatchers|CreateEvent :!event Custom|Firing))
  (Development|PrintString (Utilities|GetDisplayName _asb_high_tower_tomato_repulsors) true true "(R=0.290283,G=1.000000,B=0.000000,A=1.000000)"))

(event Custom|Muzzle_Firing_Setup
  (params (WeaponObject :type Fort%20Weapon%20Object%20Reference))
  ;[1 0.6369 0.427] Muzzle Firing Setup - Must be Replicated to Function!
  (CallFunction|MuzzleFiringMulticast WeaponObject))

(event Custom|Firing
  (params (Hand :type EDualWeaponHand%20Enum :default "LEFT") (bPersistantFire :type Boolean))
  (bind _returnvalue (CustomCharacter|GetAssociatedPlayerPawn))
  (bind _returnvalue_2 (Fortnite|GetFortPlayerControllerfromActor _returnvalue))
  (bind _returnvalue_3 (CustomCharacter|GetCharacterPartType))
  (bind _returnvalue_4 (Customization|GetSkeletalMeshforPartType _returnvalue _returnvalue_3 :!class /Script/FortniteGame.FortPlayerPawn))
  (bind _lmb_color (Variables|Default|GetLMB%20Color))
  ;[1 0.6369 0.427] Firing - IdleControl
  (if (not (GameplayTags|HasMatchingGameplayTag (CustomCharacter|GetAssociatedPlayerPawn) "(TagName=\"Granted.Athena.Power.ActivePower.IsActive\")"))
    (Utilities|FlowControl|Sequence
      (:then_0
        (if (Utilities|Enum|Equal(Enum) Hand "LEFT")
          (Components|Activation|Activate (Variables|Default|GetFX_Muzzle_NS_L) false)
          (Rendering|Material|SetVectorParameterValueonMaterials _returnvalue_4 "Emissive Color" (Math|Conversions|ToVector(LinearColor) _lmb_color))
          (else
            (Components|Activation|Activate (Variables|Default|GetFX_Muzzle_NS_R) false)
            (Rendering|Material|SetVectorParameterValueonMaterials _returnvalue_4 "Emissive Color" (Math|Conversions|ToVector(LinearColor) _lmb_color)))))
      (:then_1
        ;[1 1 1] Camera shake
        (Utilities|IsValid _returnvalue_2
          (:Is%20Valid
            (CameraShakes|StartCameraShake :self (Variables|/Script/Engine.PlayerController|GetPlayerCameraManager _returnvalue_2) :ShakeClass "/HighTower/Items/Tomato/Repulsors/Proto/HighTower_Tomato_Repulsor_CameraShake.HighTower_Tomato_Repulsor_CameraShake_C" :Scale 1.0 :PlaySpace "CameraLocal")))))))

(event Custom|Muzzle_Attachment
  (bind _returnvalue (CustomCharacter|GetSkeletalMeshForPartOnPlayerPawn))
  ;[1 0.6369 0.427] Muzzle Setups
  (Transformation|AttachComponentToComponent (Variables|Default|GetFX_Muzzle_NS_L) _returnvalue "FX_repulsor_l" "KeepRelative" "KeepRelative" "KeepRelative" true)
  (Transformation|AttachComponentToComponent (Variables|Default|GetFX_Muzzle_NS_R) _returnvalue "FX_repulsor_r" "KeepRelative" "KeepRelative" "KeepRelative" true))

(fn OnRep_OurCurrentWeapon
  (bind _asb_ranged_ashton_indigo_athena (Utilities|Casting|CastToB_Ranged_Ashton_Indigo_Athena (Variables|Default|GetOurCurrentWeapon))
    (:then
      (RangedWeaponDual|BindEventtoOnHandFired _asb_ranged_ashton_indigo_athena (EventDispatchers|CreateEvent :!event Firing))
      (Development|PrintString (Utilities|GetDisplayName _asb_ranged_ashton_indigo_athena) true true "(R=0.290283,G=1.000000,B=0.000000,A=1.000000)"))
    (:CastFailed
      (Development|PrintString "CASTFAILED" true true "(R=1.000000,G=0.000000,B=0.031557,A=1.000000)"))))

(fn ConstructionScript)
