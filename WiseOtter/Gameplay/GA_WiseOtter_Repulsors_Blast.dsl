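; Gameplay ability for the WiseOtter repulsor blast; this graph overrides the failure reason reported when activation fails.
; It casts the avatar to a Fort player pawn and checks for the Tomato RepulsorCannon (Unibeam) ability tag.
; With that tag it returns an empty reason, suppressing the failure warning because Unibeam's visual state is already obvious.
; Otherwise, the incoming reason is meant to pass through unchanged.
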(fn OverrideFailedReason
  (params (FailedReason :type Gameplay%20Tag%20Container%20Structure%20%28by%20ref%29))
  (bind _asfort_player_pawn (Utilities|Casting|CastToFortPlayerPawn (Ability|GetAvatarActorfromActorInfo))
    (:then
      (if (GameplayTags|HasMatchingGameplayTag _asfort_player_pawn "(TagName=\"Ability.HighTower.Power.Tomato.RepulsorCannon\")")
        ; FORT-780832: Unibeam has a very obvious visual state - no need to warn about the repulsors not activating in this case. In addition this was a problem for the capture team. Ideally we might also solve this by knowing whether we were holding down primary fire input and not show a failed message in that activation vs on a new press.
        (return 0)
        (else
          _)))
    (:CastFailed
      _))
  (return FailedReason))
